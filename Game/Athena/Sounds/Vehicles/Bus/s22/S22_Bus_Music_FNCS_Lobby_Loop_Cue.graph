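# Sound cue for the Season 22 FNCS Battle Bus lobby music, "This Just Got Real", pairing a one-shot intro with a looping bed.
# A Mixer layers the slightly quieter intro with the loop, which starts after a fixed delay of about half a second.
# An Enveloper at the root fades the mix in over one second.
# Output uses the Battle Bus sound class and close lobby bus-engine attenuation.
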SoundCue /Game/Athena/Sounds/Vehicles/Bus/s22/S22_Bus_Music_FNCS_Lobby_Loop_Cue
root=Enveloper_0
nodes [
 {id=Delay_0, children [WavePlayer_3], properties {delayMax=0.497, delayMin=0.497}}
 {
  id=Enveloper_0
  children [Mixer_0]
  properties {pitchCurve {editorCurveData {keys [{time=0}]}}, volumeCurve {editorCurveData {keys [<at 0 0>,<at 1 1>]}}}
 }
 {id=Mixer_0, children [Delay_0,WavePlayer_4], properties {inputVolume [1,0.75]}}
 {id=WavePlayer_3, properties {bLooping=true, soundWaveAssetPtr=&/s22_FNCS_This_Just_Got_Real_BB_Loop}}
 {id=WavePlayer_4, properties {soundWaveAssetPtr=&/s22_FNCS_This_Just_Got_Real_BB_Intro}}
]
properties {
 attenuationSettings=/Game/Sounds/Attenuation/Fort_Bus_Lobby_Engine_Athena_Close_Att
 maxDistance=12500
 soundClassObject=/Game/Sounds/FortSoundClassesAndModes/Fort_BattleBus
 volumeMultiplier=0.65
}
