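# Sound cue for the Season 22 FNCS Battle Bus music, playing "This Just Got Real" as an intro plus a loop.
# A mixer starts the one-shot intro at reduced volume alongside a fixed half-second delay that leads into the looping track.
# The cue uses the Battle Bus sound class and references the bus music attenuation asset, with overrides specifying no attenuation or spatialization.
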SoundCue /Game/Athena/Sounds/Vehicles/Bus/s22/S22_Bus_Music_FNCS_Loop_Cue
root=Mixer_0
nodes [
 {id=Delay_0, children [WavePlayer_0], properties {delayMax=0.497, delayMin=0.497}}
 {id=Mixer_0, children [Delay_0,WavePlayer_2], properties {inputVolume [1,0.75]}}
 {id=WavePlayer_0, properties {bLooping=true, soundWaveAssetPtr=&/s22_FNCS_This_Just_Got_Real_BB_Loop}}
 {id=WavePlayer_2, properties {soundWaveAssetPtr=&/s22_FNCS_This_Just_Got_Real_BB_Intro}}
]
properties {
 attenuationSettings=/Game/Sounds/Attenuation/Fort_Bus_Music_Athena_Att
 maxDistance=21000
 soundClassObject=/Game/Sounds/FortSoundClassesAndModes/Fort_BattleBus
 volumeMultiplier=0.85
 attenuationOverrides {bAttenuate=false, bSpatialize=false}
}
